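# Conversation database for the Dark Vogue NPC, supplying the spoken greeting that opens a player's conversation.
# The entry point feeds one speech node whose contextual lines cover first-ever meetings, match conversation count checks and a reaction to a specific character outfit.
# First-meeting and outfit lines take priority from named curve rows.
# The speech node then branches through link nodes to the shared NPC library quests and basic services conversations.
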ConversationDatabase /NPCLibrary/NPCs/DarkVogue/NPCLibrary_Conversation_DarkVogue
entry_points [{tag=Conversation.Entry.DarkVogue, destinations [#1]}]
nodes [
 ConversationEntryPointNode_7 {
  id=#1
  class=ConversationEntryPointNode
  outputs [#2]
  properties {entryTag {tagName=Conversation.Entry.DarkVogue}}
 }
 Speech_C_4 {
  id=#2
  class=Speech_C
  outputs [#3, #4]
  properties {
   nodeName=""
   generalConfig {
    defaultMessage="This place is so \"spooky\", sure."
    contextualMessages [
     {
      message="Welcome to the nitemare, I guess."
      contextRequirements [{requirement {class=FortControllerRequirement_IsFirstEverConversationWithNPC}}]
      priority {curve {rowName=Tandem.Speech.Priority.FirstEverInteraction}}
     }
     {
      message="This place is so \"spooky\", sure."
      contextRequirements [{requirement {class=FortControllerRequirement_CheckNPCMatchConversationCount, properties {nPCUniqueIdTag {tagName=AISpawnerData.Type.Tandem.DarkVogue}}}}]
      priority {curve {rowName=None}}
     }
     {
      message=I'm scarier than anything on this Island.
      contextRequirements [{requirement {class=FortControllerRequirement_CheckNPCMatchConversationCount, properties {nPCUniqueIdTag {tagName=AISpawnerData.Type.Tandem.DarkVogue}}}}]
      priority {curve {rowName=None}}
     }
     {
      message=I've been more scared at the grocery store.
      contextRequirements [{requirement {class=FortControllerRequirement_CheckNPCMatchConversationCount, properties {nPCUniqueIdTag {tagName=AISpawnerData.Type.Tandem.DarkVogue}}}}]
      priority {curve {rowName=None}}
     }
     {
      message="You're looking grave. That's not a pun, are you okay?"
      contextRequirements [{requirement {class=FortControllerRequirement_CheckNPCMatchConversationCount, properties {nPCUniqueIdTag {tagName=AISpawnerData.Type.Tandem.DarkVogue}}}}]
      priority {curve {rowName=None}}
     }
     {
      message=Why do you look so scared?
      contextRequirements [{requirement {
         class=FortControllerRequirement_HasCID
         properties {allowedCIds [/BRCosmetics/Athena/Items/Cosmetics/Characters/CID_527_Athena_Commando_F_StreetFashionRed]}
        }}]
      priority {curve {rowName=Tandem.Speech.Priority.OutfitReaction}}
     }
    ]
   }
  }
 }
 ConversationLinkNode_0 {id=#4, class=ConversationLinkNode, properties {remoteEntryTag {tagName=Conversation.Entry.BasicServices}}}
 ConversationLinkNode_1 {id=#3, class=ConversationLinkNode, properties {remoteEntryTag {tagName=Conversation.Entry.NPCLibrary.Quests}}}
]
